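; Ability that has an NPC toss loot for the player in an Athena (Battle Royale) match.
; After a random delay, it restricts the NPC's behavior control, plays a toss montage, then on a montage-length timer spawns loot ahead of the pawn and fires OnLootGranted.
; The loot tier is keyed by playlist tag, with Duo as fallback.
; The ability ends shortly afterwards, restoring behavior and clearing both timers.
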(event Ability|EventActivateAbility
  ;[0 0 0] Activate Ability
  (|CallParentFunction)
  (bind _returnvalue (Utilities|Time|SetTimerbyEvent (delegate Custom|TossLoot) (Math|Random|RandomFloatinRange 1.0 3.0) false true))
  (Variables|Default|SetLootTossDelayTimerHandle _returnvalue))

(event Custom|TossLoot
  ;[0 0 0] Activate Ability
  ;; Ensure the NPC can do nothing except emote
  (AI|PushBehaviorControl (Variables|InternalData|GetAIController) -2093056)
  (bind _asbp_player_pawn_tandem (Utilities|Casting|CastToBP_PlayerPawn_Tandem (Pawn|GetControlledPawn (Variables|InternalData|GetAIController))))
  ;; Play the "toss" animation, spawn loot, and end ability
  (Animation|PlayAnimMontage _asbp_player_pawn_tandem (Variables|Default|GetTossLootAnim) 1.0 "None")
  (bind _returnvalue (Utilities|Time|SetTimerbyEvent (delegate Custom|AnimFinished) (Animation|GetPlayLength (Variables|Default|GetTossLootAnim)) false true))
  (Variables|Default|SetAnimLengthTimerHandle _returnvalue))

(event Custom|AnimFinished
  ;[0 0 0] Activate Ability
  ;; Play the "toss" animation, spawn loot, and end ability
  (bind _playlistsquadtag (CallFunction|GetPlaylistSquadTag))
  (Class|AthenaFunctionLibrary|SpawnLootfromTable (Utilities|Map|Find (Variables|Default|GetPlaylistLootTiers) _playlistsquadtag) (Loot|GetLootLevel) -1 (+ (Transformation|GetActorLocation (stale Custom|TossLoot _asbp_player_pawn_tandem)) (* (Variables|Default|GetSpawnOffsetFromNPC) (Transformation|GetActorForwardVector (stale Custom|TossLoot _asbp_player_pawn_tandem)))) (Transformation|GetActorForwardVector (stale Custom|TossLoot _asbp_player_pawn_tandem)) 0 true true)
  (Default|CallOnLootGranted (stale Custom|TossLoot _asbp_player_pawn_tandem))
  (Utilities|FlowControl|Delay 2.0)
  (Ability|EndAbility))

(event Ability|EventOnEndAbility
  (params (bWasCancelled :type Boolean))
  ;[0 0 0] End ability
  (|CallParentFunction bWasCancelled)
  ;; Return NPC to normal behavior
  (AI|PopBehaviorControl (Variables|InternalData|GetAIController))
  (CallFunction|ClearTimerHandle (Variables|Default|GetLootTossDelayTimerHandle))
  (CallFunction|ClearTimerHandle (Variables|Default|GetAnimLengthTimerHandle)))

(event Custom|ClearTimerHandle
  (params (TimerHandle :type Timer%20Handle%20Structure))
  ;[0 0 0] Helper events
  (if (Utilities|Time|IsValidTimerHandle)
    (Utilities|Time|ClearandInvalidateTimerbyHandle TimerHandle)))

(fn GetPlaylistSquadTag
  (Utilities|FlowControl|Sequence
    (:then_0
      (if (ScalableFloat|GetValueasBool (Variables|Default|GetHFRow%20UseDefaultLootTier))
        (bind _keypin (Utilities|Map|ForEachLoop(Map) (Variables|Default|GetPlaylistLootTiers)))
        (if (GameplayTags|TagContainer|CompareTagContainertoOtherTags _keypin (Class|FortGameStateAthena|GetAthenaPlaylistContextTags (Fortnite|Athena|GameState|GetGameStateAthena)))
          (return _keypin))))
    (:then_1
      (return "(TagName=\"Athena.Playlist.Duo\")"))))
